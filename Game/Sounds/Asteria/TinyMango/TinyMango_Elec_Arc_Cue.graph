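# Sound cue for the TinyMango electric arc effect, classed as an Athena ambient prop sound.
# A Random node picks among ten equally weighted arc wave variations and is excluded from branch culling.
# A Modulator at the root varies each play, with a lowered minimum volume for more level variation.
# The cue uses the shared TinyMango electric attenuation asset and a reduced overall volume.
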SoundCue /Game/Sounds/Asteria/TinyMango/TinyMango_Elec_Arc_Cue
root=Modulator_0
nodes [
 {id=Modulator_0, children [Random_1], properties {volumeMin=0.8}}
 {
  id=Random_1
  children [
   WavePlayer_6
   WavePlayer_7
   WavePlayer_8
   WavePlayer_9
   WavePlayer_10
   WavePlayer_11
   WavePlayer_12
   WavePlayer_13
   WavePlayer_14
   WavePlayer_15
  ]
  properties {bShouldExcludeFromBranchCulling=true, weights [1,1,1,1,1,1,1,1,1,1]}
 }
 {id=WavePlayer_10, properties {soundWaveAssetPtr=&/TinyMango_Elec_Arc_05}}
 {id=WavePlayer_11, properties {soundWaveAssetPtr=&/TinyMango_Elec_Arc_06}}
 {id=WavePlayer_12, properties {soundWaveAssetPtr=&/TinyMango_Elec_Arc_07}}
 {id=WavePlayer_13, properties {soundWaveAssetPtr=&/TinyMango_Elec_Arc_08}}
 {id=WavePlayer_14, properties {soundWaveAssetPtr=&/TinyMango_Elec_Arc_09}}
 {id=WavePlayer_15, properties {soundWaveAssetPtr=&/TinyMango_Elec_Arc_10}}
 {id=WavePlayer_6, properties {soundWaveAssetPtr=&/TinyMango_Elec_Arc_01}}
 {id=WavePlayer_7, properties {soundWaveAssetPtr=&/TinyMango_Elec_Arc_02}}
 {id=WavePlayer_8, properties {soundWaveAssetPtr=&/TinyMango_Elec_Arc_03}}
 {id=WavePlayer_9, properties {soundWaveAssetPtr=&/TinyMango_Elec_Arc_04}}
]
properties {
 attenuationSettings=&/TinyMango_Elec_Att
 maxDistance=3600
 soundClassObject=&Fort_Ambient_Props_Athena
 volumeMultiplier=0.752
}
