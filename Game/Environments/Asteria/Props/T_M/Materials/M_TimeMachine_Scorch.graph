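# Translucent deferred decal material for scorch marks on the Asteria time machine prop, with a pulsing orange glow.
# A time-driven sine, bias-scaled into a narrow range, subtly blends two emissive color parameters, staying close to the first.
# That blend is multiplied by an emissive texture and a strength parameter.
# A separate decal texture sits alongside base color, roughness and specular parameters.
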Material /Game/Environments/Asteria/Props/T_M/Materials/M_TimeMachine_Scorch
graph {
 properties {blendMode=BLEND_Translucent, materialDomain=MD_DeferredDecal, pixelDepthOffsetMode=PDOM_Legacy, usedShadingModels=Default Lit}
 nodes [
  n1 = VectorParameter() {parameterName=EmissColor1, defaultValue=<rgba 0.9479 0.2594 0.0402 1>}
  n2 = VectorParameter() {parameterName=EmissColor2, defaultValue=<rgba 1 0.5556 0.0746 1>}
  n3 = ConstantBiasScale(None=Sine(None=Time())) {scale=0.1}
  Multiply(
   A=Multiply(
    A=LinearInterpolate(A=@n1.RGB, B=@n2.RGB, Alpha=@n3)
    B=TextureSample() {texture=/Game/Environments/Asteria/Props/T_M/Textures/T_TimeMachine_Scorch_Decal_E}.RGB
   )
   B=ScalarParameter() {defaultValue=20, parameterName=EmissiveStrength}
  )
  ScalarParameter() {defaultValue=0.8, parameterName=Roughness}
  ScalarParameter() {defaultValue=0.1, parameterName=Spec}
  TextureSample() {texture=/Game/Environments/Asteria/Props/T_M/Textures/T_TimeMachine_Scorch_Decal}
  VectorParameter() {parameterName=BaseColor}
 ]
}
